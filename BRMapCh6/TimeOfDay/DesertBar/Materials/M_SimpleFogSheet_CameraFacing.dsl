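; Fog sheet Blueprint actor, stored under the DesertBar time-of-day content, that puts a configurable simple fog material on a plane.
; Its construction script creates a dynamic instance of M_SimpleFogSheet on the Plane component, then passes in variables for color, brightness, depth fade, near fade and panning speed.
; BeginPlay, overlap and Tick events are empty, and the graph holds no camera-facing logic despite the asset name.
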(event EventBeginPlay)

(event Collision|EventActorBeginOverlap
  (params (OtherActor :type Actor%20Object%20Reference)))

(event EventTick
  (params (DeltaSeconds :type Float%20%28single-precision%29)))

(fn ConstructionScript
  (bind _returnvalue (Rendering|Material|CreateDynamicMaterialInstance (Variables|Default|GetPlane) 0 "/Game/Environments/TechArt/FogSheet_Simple/M_SimpleFogSheet.M_SimpleFogSheet" "None" :!class /Script/Engine.PrimitiveComponent))
  (Rendering|Material|SetVectorParameterValue _returnvalue "FogColor" (Variables|Default|GetFog%20Color) :!class /Script/Engine.MaterialInstanceDynamic)
  (Rendering|Material|SetScalarParameterValue _returnvalue "FogBrightness" (Variables|Default|GetFog%20Brightness) :!class /Script/Engine.MaterialInstanceDynamic)
  (Rendering|Material|SetScalarParameterValue _returnvalue "DepthFadeDistance" (Variables|Default|GetDepth%20Fade%20Distance) :!class /Script/Engine.MaterialInstanceDynamic)
  (Rendering|Material|SetScalarParameterValue _returnvalue "NearFadeOffset" (Variables|Default|GetNear%20Fade%20Offset) :!class /Script/Engine.MaterialInstanceDynamic)
  (Rendering|Material|SetScalarParameterValue _returnvalue "NearFadeBlend" (Variables|Default|GetNear%20Fade%20Blend) :!class /Script/Engine.MaterialInstanceDynamic)
  (Rendering|Material|SetScalarParameterValue _returnvalue "PanningSpeed" (Variables|Default|GetPanning%20Speed) :!class /Script/Engine.MaterialInstanceDynamic))
